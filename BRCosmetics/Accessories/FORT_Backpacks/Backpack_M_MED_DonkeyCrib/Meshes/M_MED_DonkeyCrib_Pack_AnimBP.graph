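# Animation Blueprint for the DonkeyCrib backpack that toggles a material element based on the glider state.
# The anim graph only passes its input pose through, so the logic sits in the event graph.
# It loops over the part mesh's material slots, getting dynamic instances and setting "Hide Element 10" to 1 when an opened glider has the DonkeyCribGlider cosmetic meta tag, otherwise 0.
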AnimBlueprint /BRCosmetics/Accessories/FORT_Backpacks/Backpack_M_MED_DonkeyCrib/Meshes/M_MED_DonkeyCrib_Pack_AnimBP
target_skeleton=&/M_MED_DonkeyCrib_Pack_Skeleton
graphs [
 AnimGraph {nodes [Output Pose(Result=Input Pose()) {layerGroup=None}]}
 EventGraph {nodes [
   Event BlueprintUpdateAnimation()
   n1 = GetPartSkeletalMeshComponent()
   > n2 = Decrement Int(Value=GetNumMaterials!(self:Primitive Component Object Reference=@n1) -> (ReturnValue:int=0))
   > n3 = For Loop(LastIndex=@n2)
   > n4 = Get Dynamic Material Instance(SkeletalMeshComponent=@n1, ElementIndex=@n3)
   > n5 = Branch(Condition=GliderItemDefinitionHasAnyMetaTag!(Tags:Array of Gameplay Tag Structures=Make Array("[0]"=(TagName="Cosmetics.Metatags.DonkeyCribGlider")), bRequiresOpened:bool=true) -> (ReturnValue:bool))
   > SetScalarParameterValue(self=@n4, ParameterName=Hide Element 10, Value=1)
   SetScalarParameterValue(execute=@n5.else, self=@n4, ParameterName=Hide Element 10, Value=0)
   TryGetPawnOwner()
  ]}
]
variables []
